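# Demo animation Blueprint for the prototype flotation balloon's back-knot rope, attached to a player pawn.
# A cosmetic-part callback offsets the Handle bone from the backpack_br socket, feeding a half-blended Anim Dynamics string chain, then balloon_attach takes the pawn's world rotation.
# Update caches and prints the parent pawn once, then refreshes rotation and registers that callback every tick.
# Debug-arrow and speed-alpha logic is unwired.
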AnimBlueprint /Game/Weapons/Prototype/Flotation_Balloon/Demo/SK_Weapon_Floatation_Balloon_BackKnot_Dyn_AnimBP
target_skeleton=&/SK_Weapon_Floatation_Balloon_Rope_Skeleton
graphs [
 AnimGraph {nodes [
   n1 = "Transform (Modify) Bone - Bone: Handle"!(
    ComponentPose=Local To Component(LocalPose=Local Space Ref Pose())
    Alpha=1
    Translation=Get backpack_position() -> (backpack_position:Vector)
    Rotation=Get backpack_rotation() -> (backpack_rotation:Rotator)
   ) -> (Pose) {
    rotationMode=BMM_Replace
    rotationSpace=BCS_BoneSpace
    translationMode=BMM_Additive
    translationSpace=BCS_ParentBoneSpace
    alphaBoolBlend {blendOption=Linear}
    boneToModify {boneName=Handle}
   }
   n2 = "Anim Dynamics - Chain: dyn_string_01 -> dyn_string_03"(ComponentPose=@n1, Alpha=0.5, ExternalForce="0,0,0") {
    angularDampingOverride=0.965
    bChain=true
    bGravityOverrideInSimSpace=true
    bOverrideAngularDamping=true
    bOverrideLinearDamping=true
    gravityScale=0.5
    lODThreshold=1
    linearDampingOverride=0.965
    numSolverIterationsPostUpdate=4
    numSolverIterationsPreUpdate=16
    alphaBoolBlend {blendOption=Linear}
    boundBone {boneName=dyn_string_01}
    chainEnd {boneName=dyn_string_03}
    physicsBodyDefinitions [{
      sphereCollisionRadius=0
      boundBone {boneName=dyn_string_01}
      boxExtents=<5 5 5>
      constraintSetup {angularTargetAxis=AxisZ, twistAxis=AxisZ, angularLimitsMax=<5 30 30>, angularLimitsMin=<-90 -30 -30>, angularTarget=<0 0 0>}
      localJointOffset=<0 0 -0.5>
     }]
   }
   n3 = "Transform (Modify) Bone - Bone: balloon_attach"(ComponentPose=@n2, Alpha=1, Translation="0,0,0", Rotation=Get socketUpVector() -> (socketUpVector:Rotator), Scale="1,1,1") {
    rotationMode=BMM_Replace
    rotationSpace=BCS_WorldSpace
    translationSpace=BCS_WorldSpace
    alphaBoolBlend {blendOption=Linear}
    boneToModify {boneName=balloon_attach}
   }
   Output Pose(Result=Component To Local(ComponentPose=@n3)) {layerGroup=None}
  ]}
 EventGraph {nodes [
   n1 = WhenBody!() -> (
    OutputDelegate
    then
    Pawn:Fort Player Pawn Object Reference
    Part:Custom Character Part Object Reference
    MeshPart:Skeletal Mesh Component Object Reference
    PartType:EFortCustomPartType Enum=Head
   ) {is=title K2Node_CustomEvent}
   n2 = BreakTransform(InTransform=GetSocketTransform(self=@n1.MeshPart, InSocketName=backpack_br, TransformSpace=RTS_ParentBoneSpace))
   > Set backpack_position(backpack_position:Vector=vector + vector(A:Vector=@n2, B:Vector="5,10,0") -> (ReturnValue:Vector) {is=title K2Node_CommutativeAssociativeBinaryOperator}) -> (Output_Get:Vector)
   > Set backpack_rotation(backpack_rotation:Rotator=Combine Rotators!(A:Rotator=@n2.Rotation, B:Rotator="90, 0, 0") -> (ReturnValue:Rotator)) -> (Output_Get:Rotator)
   Event BlueprintUpdateAnimation()
   > n3 = Sequence()
   > Do Once()
   n4 = GetAttachParentActor!(self:Actor Object Reference=GetOwningActor()) -> (ReturnValue:Actor Object Reference)
   > n5 = Cast To FortPlayerPawnAthena(Object=@n4)
   > Set player_pawn(player_pawn:Fort Player Pawn Athena Object Reference=@n5) -> (Output_Get:Fort Player Pawn Athena Object Reference)
   > PrintString!(execute, InString=GetDisplayName(Object=@n4), bPrintToScreen, bPrintToLog, TextColor, Duration=10, Key) -> (then)
   Set socketUpVector(
    execute=@n3.then_1
    socketUpVector:Rotator=Get Actor Rotation!(self:Actor Object Reference=Get player_pawn() -> (player_pawn:Fort Player Pawn Athena Object Reference)) -> (ReturnValue:Rotator) {is=title K2Node_CallFunction}
   ) -> (Output_Get:Rotator)
   > WhenCosmeticSpecificPartApplied!(
    execute
    self:Fort Player Pawn Object Reference=Get player_pawn()
    Delegate:Delegate=@n1
    SkipNullMesh:bool=false
    PartType:EFortCustomPartType Enum
   ) -> (then, DelegateHandleController_Out:Delegate Handle Controller Structure)
   n6 = GetSocketLocation!(self:Scene Component Object Reference=GetPartSkeletalMeshComponent(), InSocketName:Name=balloon_attach) -> (ReturnValue:Vector)
   n7 = Get Rotation X Vector!(InRot:Rotator="0, 0, 0") -> (ReturnValue:Vector) {is=title K2Node_CallFunction}
   DrawDebugArrow!(
    execute
    LineStart:Vector=@n6
    LineEnd:Vector=vector + vector(A=@n6, B=vector * float(A=@n7, B=100))
    ArrowSize:float=0
    LineColor:Linear Color Structure=(R=0.9462,G=0,B=1,A=1)
    Duration:float=0
    Thickness:float=2
    DepthPriority:EDrawDebugSceneDepthPriorityGroup Enum=World
   ) -> (then)
   Event BlueprintInitializeAnimation()
   Reroute Node(InputPin:Wildcard) -> (OutputPin:Wildcard)
   n8 = Get player_pawn()
   n9 = GetVelocity!(self:Actor Object Reference=@n8) -> (ReturnValue:Vector)
   n10 = Get Actor Rotation(self=@n8)
   n11 = BreakVector(InVec=vector * vector(A:Vector=Get Rotation X Vector(InRot=@n10), B:Vector=@n9) -> (ReturnValue:Vector) {is=title K2Node_CommutativeAssociativeBinaryOperator})
   n12 = float + float(A=@n11, B=@n11.Y) {is=title K2Node_CommutativeAssociativeBinaryOperator}
   n13 = "Not Equal (Float)"!(A:double=BreakVector(InVec=@n9).Z, B:double=0) -> (ReturnValue:bool) {is=title K2Node_CallFunction}
   Set pawn_speed_alpha(pawn_speed_alpha:double=float * float(A=MapRangeClamped(Value=@n12, InRangeA=0, InRangeB=500), B=SelectFloat(A=0, B=1, bPickA=@n13))) -> (Output_Get:double)
  ]}
]
variables [
 backpack_position {type=Vector, default=<-20 0 0>}
 backpack_rotation {type=Rotator}
 pawn_speed_alpha {type=number}
 player_pawn {type=FortPlayerPawnAthena}
 socketUpVector {type=Rotator}
]
