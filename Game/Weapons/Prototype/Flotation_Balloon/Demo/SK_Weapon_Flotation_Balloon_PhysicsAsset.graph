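# Physics asset for the prototype Flotation Balloon weapon's skeletal mesh, in a Demo folder, giving it simple collision bodies.
# A large capsule on the balloon_rbd bone is offset upward as the balloon volume, and a small sphere sits on the root bone.
# No constraints link the two bodies, so the asset itself defines no joint between balloon and root.
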PhysicsAsset /Game/Weapons/Prototype/Flotation_Balloon/Demo/SK_Weapon_Flotation_Balloon_PhysicsAsset
bodies {
 balloon_rbd [{shapeType=Capsule, length=18, radius=25.75, center=<0 0 35.6511>, rotation=<rot 0 -90 0>}]
 root [{shapeType=Sphere, radius=5, center=<0 0 10.1029>}]
}
constraints []
