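# Animation Blueprint for the prototype Flotation Balloon, posing the balloon and its string with procedural secondary motion.
# The attached parent's pose gets world-space root orientation, side tilt, a string spring and alpha-driven head-back correctives, then is cached.
# Two velocity-weighted trail chains read that cache, blending to the shorter when moving backward, before a look-at aims balloon_aim along the string.
# The EventGraph is empty; nearly all driving variables are declared outside this file.
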AnimBlueprint /Game/Weapons/Prototype/Flotation_Balloon/Demo/SK_Weapon_Flotation_Balloon_Skeleton_AnimBP
target_skeleton=&/SK_Weapon_Flotation_Balloon_Skeleton
graphs [
 AnimGraph {nodes [
   n1 = Get PawnZVelCheck() -> (PawnZVelCheck:float)
   n2 = Trail controller - Bone: balloon_rbd(ComponentPose=Local To Component(LocalPose=Use cached pose 'SavedPose13970'()), Alpha=@n1) {
    chainBoneAxis=Z
    chainLength=15
    maxDeltaTime=0.0333
    relaxationSpeedScale=3
    alphaScaleBiasClamp {bInterpResult=true, interpSpeedIncreasing=1}
    rotationLimits [
     {limitMax=<180 180 180>}
     {limitMax=<180 180 180>}
     {limitMax=<180 180 180>}
     {limitMax=<180 180 180>}
     {limitMax=<180 180 180>}
     {limitMax=<180 180 180>}
     {limitMax=<180 180 180>}
     {limitMax=<180 180 180>}
     {limitMax=<180 180 180>}
     {limitMax=<180 180 180>}
     {limitMax=<180 180 180>}
     {limitMax=<180 180 180>}
     {limitMax=<180 180 180>}
     {limitMax=<180 180 180>}
     {limitMax=<180 180 180>}
    ]
    rotationOffsets [<0 0 0>, <0 0 0>, <0 0 0>, <0 0 0>, <0 0 0>, <0 0 0>, <0 0 0>, <0 0 0>, <0 0 0>, <0 0 0>, <0 0 0>, <0 0 0>, <0 0 0>, <0 0 0>, <0 0 0>]
    trailBone {boneName=balloon_rbd}
    trailRelaxationSpeed {editorCurveData {keys [
       {arriveTangent=-10.9965, interpMode=RCIM_Linear, time=0, value=10}
       {arriveTangent=-14.6479, interpMode=RCIM_Cubic, leaveTangent=-14.6479, tangentMode=RCTM_User, time=1, value=7}
      ]}}
   }
   n3 = Component To Local(ComponentPose=@n2)
   n4 = Trail controller - Bone: balloon_rbd(ComponentPose=Local To Component(LocalPose=Use cached pose 'SavedPose13970'()), Alpha=@n1) {
    chainBoneAxis=Z
    chainLength=9
    maxDeltaTime=0.0333
    relaxationSpeedScale=4
    alphaScaleBiasClamp {bInterpResult=true, interpSpeedIncreasing=1}
    rotationLimits [
     {limitMax=<180 180 180>}
     {limitMax=<180 180 180>}
     {limitMax=<180 180 180>}
     {limitMax=<180 180 180>}
     {limitMax=<180 180 180>}
     {limitMax=<180 180 180>}
     {limitMax=<180 180 180>}
     {limitMax=<180 180 180>}
     {limitMax=<180 180 180>}
    ]
    rotationOffsets [<0 0 0>, <0 0 0>, <0 0 0>, <0 0 0>, <0 0 0>, <0 0 0>, <0 0 0>, <0 0 0>, <0 0 0>]
    trailBone {boneName=balloon_rbd}
    trailRelaxationSpeed {editorCurveData {keys [
       {arriveTangent=-10.9965, interpMode=RCIM_Linear, time=0, value=10}
       {arriveTangent=-14.6479, interpMode=RCIM_Cubic, leaveTangent=-14.6479, tangentMode=RCTM_User, time=1, value=7}
      ]}}
   }
   n5 = Component To Local(ComponentPose=@n4)
   n6 = Get bIsMovingBackward() -> (bIsMovingBackward:bool)
   n7 = Look At - Bone: balloon_aim!(
    ComponentPose=Local To Component(LocalPose=Blend Poses by bool(BlendPose_0=@n5, BlendPose_1=@n3, BlendTime_0=0.5, BlendTime_1=0.5, bActiveValue=@n6) {blendType=Linear})
    Alpha=Get AimAlpha() -> (AimAlpha:float)
    LookAtLocation:Vector="0,0,0"
   ) -> (Pose) {
    alphaBoolBlend {blendOption=Linear}
    boneToModify {boneName=balloon_aim}
    lookAtLocation {x=0}
    lookAtTarget {boneReference {boneName=balloon_string_14}, socketReference {socketName=rootSocket}}
    lookAt_Axis {axis {y=0, z=-1}}
    lookUp_Axis {bInLocalSpace=false, axis {x=0}}
   }
   Output Pose(Result=Component To Local(ComponentPose=@n7)) {layerGroup=None}
   n8 = "Transform (Modify) Bone - Bone: root"!(
    ComponentPose=Local To Component(LocalPose=Copy Pose From Mesh() {bUseAttachedParent=true})
    Alpha
    Rotation=Get SocketUpRotator() -> (SocketUpRotator:Rotator)
   ) -> (Pose) {
    rotationMode=BMM_Replace
    rotationSpace=BCS_WorldSpace
    translationSpace=BCS_ParentBoneSpace
    alphaBoolBlend {blendOption=Linear}
    alphaScaleBiasClamp {bInterpResult=true}
    boneToModify {boneName=root}
   }
   n9 = "Transform (Modify) Bone - Bone: root"(ComponentPose=@n8, Rotation=Get BalloonSideRot() -> (BalloonSideRot:Rotator)) {
    rotationMode=BMM_Additive
    translationSpace=BCS_ParentBoneSpace
    alphaBoolBlend {blendOption=Linear}
    alphaScaleBiasClamp {bInterpResult=true}
    boneToModify {boneName=root}
   }
   n10 = Spring controller - Bone: balloon_string_1!(ComponentPose:Component Space Pose Link Structure=@n9, Alpha:float=Get SpringAlpha() -> (SpringAlpha:float)) -> (Pose:Component Space Pose Link Structure) {
    is=title AnimGraphNode_SpringBone
    bRotateX=true
    bRotateY=true
    bTranslateX=false
    bTranslateY=false
    bTranslateZ=false
    springDamping=0.1
    springStiffness=100
    alphaBoolBlend {blendOption=Linear}
    alphaScaleBiasClamp {bInterpResult=true}
    springBone {boneName=balloon_string_1}
   }
   n11 = Get HeadbackPoseAlpha() -> (HeadbackPoseAlpha:float)
   n12 = "Transform (Modify) Bone - Bone: root"(ComponentPose=@n10, Alpha=@n11, Rotation="49,0,0") {rotationMode=BMM_Additive, alphaBoolBlend {blendOption=Linear}, boneToModify {boneName=root}, rotation {pitch=49}}
   n13 = Get head_corrective_rotation() -> (head_corrective_rotation:Rotator)
   n14 = "Transform (Modify) Bone - Bone: balloon_string_4"!(
    ComponentPose="Transform (Modify) Bone - Bone: balloon_rot_fix"!(
     ComponentPose="Transform (Modify) Bone - Bone: balloon_string_3"!(ComponentPose=@n12, Alpha=@n11, Rotation=@n13) -> (Pose) {rotationMode=BMM_Additive, alphaBoolBlend {blendOption=Linear}, boneToModify {boneName=balloon_string_3}}
     Alpha=@n11
     Rotation=@n13
    ) -> (Pose) {rotationMode=BMM_Additive, alphaBoolBlend {blendOption=Linear}, boneToModify {boneName=balloon_rot_fix}}
    Alpha=@n11
    Rotation=@n13
   ) -> (Pose) {rotationMode=BMM_Additive, alphaBoolBlend {blendOption=Linear}, boneToModify {boneName=balloon_string_4}}
   n15 = "Transform (Modify) Bone - Bone: balloon_string_2"!(
    ComponentPose="Transform (Modify) Bone - Bone: balloon_string_5"!(
     ComponentPose="Transform (Modify) Bone - Bone: balloon_string_1"!(ComponentPose=@n14, Alpha=@n11, Rotation=@n13) -> (Pose) {rotationMode=BMM_Additive, alphaBoolBlend {blendOption=Linear}, boneToModify {boneName=balloon_string_1}}
     Alpha=@n11
     Rotation=@n13
    ) -> (Pose) {rotationMode=BMM_Additive, alphaBoolBlend {blendOption=Linear}, boneToModify {boneName=balloon_string_5}}
    Alpha=@n11
    Rotation=@n13
   ) -> (Pose) {rotationMode=BMM_Additive, alphaBoolBlend {blendOption=Linear}, boneToModify {boneName=balloon_string_2}}
   Save cached pose 'SavedPose13970'(Pose=Component To Local(ComponentPose="Transform (Modify) Bone - Bone: balloon_string_6"!(ComponentPose=@n15, Alpha=@n11, Rotation=@n13) -> (Pose) {rotationMode=BMM_Additive, alphaBoolBlend {blendOption=Linear}, boneToModify {boneName=balloon_string_6}})) {cacheName=SavedPose13970}
  ]}
 EventGraph {links [], nodes []}
]
variables [head_corrective_rotation {type=Rotator, default=<rot -7 0 0>}]
